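# M_StoreBGBright01 is an unlit material, apparently for a bright background element in the front-end store, judging by its name and folder.
# Its graph holds a single vector parameter, BaseColor2, defaulting to an HDR blue well above 1, which material instances can override.
# The file shows no textures, math or connection from that parameter to a material output.
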Material /Game/FrontEnd/Store/Background/M_StoreBGBright01
graph {
 properties {bFullyRough=true, opacityMaskClipValue=0.1, pixelDepthOffsetMode=PDOM_Legacy, shadingModel=MSM_Unlit, usedShadingModels=Unlit}
 nodes [VectorParameter() {parameterName=BaseColor2, defaultValue=<rgba 5.5598 7.1367 20 1>}]
}
