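# Unlit translucent material in the front-end store backgrounds, named for darkening the background with a dark tint Color parameter.
# One lerp is driven by the BGChange scalar from the shared pinata parameter collection, the other by scalar parameters including BGOpacity.
# A UseBPToDarken static switch is present, though its wiring and the material output connections are not shown.
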Material /Game/FrontEnd/Store/Background/M_StoreBGDarken
graph {
 properties {
  bUseTranslucencyVertexFog=false
  blendMode=BLEND_Translucent
  pixelDepthOffsetMode=PDOM_Legacy
  shadingModel=MSM_Unlit
  translucencyPass=MTP_BeforeDOF
  usedShadingModels=Unlit
 }
 nodes [
  Constant() {r=0.8}
  Constant()
  LinearInterpolate(
   A=Constant()
   B=Constant() {r=0.8}
   Alpha=CollectionParameter() {collection=/Game/FrontEnd/Store/Pinata/Assets/Shared/MPC_Pinata, parameterName=BGChange}
  )
  LinearInterpolate(
   A=ScalarParameter() {parameterName=sssss}
   B=ScalarParameter() {defaultValue=0.7, parameterName=aaaaaa, sliderMin=1}
   Alpha=ScalarParameter() {defaultValue=1, parameterName=BGOpacity}
  )
  StaticSwitchParameter() {parameterName=UseBPToDarken}
  VectorParameter() {parameterName=Color, defaultValue=<rgba 0.024 0.0204 0.025 1>}
 ]
}
