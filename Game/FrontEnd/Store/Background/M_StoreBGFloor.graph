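# Lit floor material for the front-end store background, with switchable diffuse texture, vertex-color darkening and edge fade.
# BaseColor is optionally multiplied by the texture's green channel and inverted vertex blue, then blended toward Fadecolor at both U edges.
# A fake reflection samples a vista texture by scaled reflection vector, masked by those channels and added to an ambient term.
# Shading path switches carry color, metal and roughness, alongside separate mobile-named terms.
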Material /Game/FrontEnd/Store/Background/M_StoreBGFloor
graph {
 properties {
  opacityMaskClipValue=0.1
  pixelDepthOffsetMode=PDOM_Legacy
  previewMesh=&/S_StoreBG06
  usedShadingModels=Default Lit
 }
 nodes [
  n1 = OneMinus(None=VertexColor().B)
  n2 = TextureSampleParameter2D(UVs=TextureCoordinate()) {parameterName=DiffuseTexture, samplerType=SAMPLERTYPE_LinearColor, texture=&/T_StoreFloor_A}
  n3 = StaticSwitchParameter(True=@n2.G, False=Constant() {r=1}) {defaultValue=true, parameterName=UseDiffuse}
  n4 = Constant() {r=0.55}
  n5 = MaterialFunctionCall(Float3=ReflectionVectorWS()) {materialFunction=&BreakOutFloat3Components}
  n6 = Multiply(A=Constant() {r=0.7}, B=@n5.R) {constB=2}
  n7 = Multiply(A=@n5.G, B=Constant() {r=3}) {constB=2}
  n8 = Constant() {r=0.05}
  n9 = TextureSample(UVs=MaterialFunctionCall(X=Add(A=@n4, B=@n6), Y=Add(A=@n7, B=@n8)) {materialFunction=&MakeFloat2}.Result) {texture=&/T_StoreBGVista}
  n10 = ScalarParameter() {defaultValue=0.3, parameterName=MobileReflectionAmount}
  n11 = VectorParameter() {parameterName=BaseColor, defaultValue=<rgba 1 1 1 1>}
  n12 = StaticSwitchParameter(True=Multiply(A=@n2.G, B=@n11.RGB), False=@n11.RGB) {defaultValue=true, parameterName=UseDiffuse}
  n13 = StaticSwitchParameter(True=Multiply(A=@n1, B=@n12), False=@n12) {defaultValue=true, parameterName=MultiplyVertColorDiffuse}
  n14 = ComponentMask(None=TextureCoordinate()) {r=true}
  n15 = LinearInterpolate(A=Constant() {r=1.5}, B=Constant() {r=-5}, Alpha=Multiply(A=@n14, B=OneMinus(None=@n14)))
  n16 = StaticSwitchParameter(
   True=LinearInterpolate(A=@n13, B=VectorParameter() {parameterName=Fadecolor, defaultValue=<rgba 0.3837 0.3783 0.39 1>}.RGB, Alpha=Clamp(None=@n15))
   False=@n13
  ) {parameterName=UseFadeColor}
  Add(
   A=Multiply(A=Multiply(A=@n1, B=@n3), B=Multiply(A=@n9.RGB, B=@n10))
   B=Multiply(A=@n16, B=ScalarParameter() {defaultValue=0.05, parameterName=MobileAmbient})
  )
  Constant() {r=0.4}
  Multiply(A=@n16, B=ScalarParameter() {parameterName=MobileDark})
  ScalarParameter() {parameterName=Spec}
  ScalarParameter() {defaultValue=1, parameterName=MobileMetal}
  ShadingPathSwitch(Default=@n16) {desc=Migrated from FeatureLevelSwitch}
  ShadingPathSwitch(Default=ScalarParameter() {defaultValue=0.5, parameterName=Metal}) {desc=Migrated from FeatureLevelSwitch}
  ShadingPathSwitch(Default=ScalarParameter() {parameterName=Rough}) {desc=Migrated from FeatureLevelSwitch}
  ShadingPathSwitch(Default=Constant()) {desc=Migrated from FeatureLevelSwitch}
 ]
}
