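# Translucent unlit glass material for the front-end store background that fakes window reflections.
# A cubemap of a command center, sampled along an offset reflection vector, is contrast-shaped, heavily brightened and partly desaturated.
# A remapped, desaturated window reflection texture masks it, and a vertical UV gradient fades the result.
# A separate U-based parabola builds a clamped edge-to-center gradient; no links to material outputs are shown.
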Material /Game/FrontEnd/Store/Background/M_StoreBGGlass
graph {
 properties {
  blendMode=BLEND_Translucent
  opacityMaskClipValue=0.1
  pixelDepthOffsetMode=PDOM_Legacy
  shadingModel=MSM_Unlit
  translucencyPass=MTP_BeforeDOF
  usedShadingModels=Unlit
 }
 nodes [
  n1 = Constant() {r=-0.6}
  n2 = Constant() {r=3.5}
  n3 = ComponentMask(None=TextureCoordinate()) {r=true}
  n4 = Multiply(A=OneMinus(None=@n3), B=@n3)
  LinearInterpolate(A=Constant() {r=1}, B=Constant() {r=-0.3}, Alpha=Clamp(None=LinearInterpolate(A=@n1, B=@n2, Alpha=@n4)))
  n5 = Constant3Vector() {constant=<rgba 0 -0.5 0 1>}
  MaterialExpressionTextureSampleParameterCube_1 = TextureSampleParameterCube!(UVs=Add(A=ReflectionVectorWS(), B=ComponentMask(None=@n5) {b=true, g=true, r=true}), Apply View MipBias) {
   parameterName=MaterialExpressionTextureSampleParameterCube_1
   samplerType=SAMPLERTYPE_LinearColor
   texture=&/TC_CommandCenterHDR
  }
  n6 = Multiply(A=Power(Base=@MaterialExpressionTextureSampleParameterCube_1.RGB, Exp=Constant() {r=1.25}), B=Constant() {r=100})
  n7 = Constant() {r=0.6}
  n8 = TextureCoordinate()
  n9 = Add(A=Constant() {r=0.25}, B=Multiply(A=ComponentMask(None=@n8) {r=true}, B=Constant() {r=0.75}))
  n10 = Multiply(A=ComponentMask(None=@n8) {g=true}, B=Constant() {r=5})
  n11 = Desaturation(None=TextureSample(UVs=AppendVector(A=@n9, B=@n10)) {texture=&/T_WindowReflection}.RGB)
  n12 = Constant() {r=2}
  n13 = Constant() {r=-0.5}
  n14 = Constant() {r=3}
  n15 = Add(A=ComponentMask(None=TextureCoordinate()) {g=true}, B=Constant() {r=-0.7})
  Multiply(
   A=LinearInterpolate(B=Desaturation(None=@n6, Fraction=@n7), Alpha=Power(Base=@n11, Exp=@n12)) {constB=0}
   B=Clamp(None=LinearInterpolate(A=@n13, B=@n14, Alpha=@n15))
  )
 ]
}
