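# Default Lit metal material for the front-end store background, with a faked environment reflection.
# A tiled club-floor texture's green channel is contrast-remapped, tinted by a horizontal center-to-edge Fadecolor gradient and masked by inverted vertex-color blue.
# Scaled, offset reflection-vector components sample the store vista texture, which is masked and added to the ambient-scaled color under Mobile-named parameters.
# Shading path switches, migrated from feature-level switches, carry the tinted color and simple surface values.
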Material /Game/FrontEnd/Store/Background/M_StoreBGMetal
graph {
 properties {opacityMaskClipValue=0.1, pixelDepthOffsetMode=PDOM_Legacy, usedShadingModels=Default Lit}
 nodes [
  n1 = VectorParameter() {parameterName=Fadecolor, defaultValue=<rgba 0.5989 0.5983 0.62 1>}
  n2 = ComponentMask(None=TextureCoordinate()) {r=true}
  n3 = OneMinus(None=@n2)
  n4 = Clamp(None=LinearInterpolate(
    A=ScalarParameter() {defaultValue=1.5, parameterName=FadeColorH}
    B=ScalarParameter() {defaultValue=-5, parameterName=FadeColorL}
    Alpha=Multiply(A=@n2, B=@n3)
   ))
  n5 = Multiply(A=TextureCoordinate(), B=ScalarParameter() {defaultValue=2.5, parameterName=Tile})
  n6 = Clamp(None=LinearInterpolate(
    A=ScalarParameter() {defaultValue=-0.2, parameterName=A}
    B=ScalarParameter() {defaultValue=6, parameterName=B}
    Alpha=TextureSample(UVs=@n5) {texture=/Game/Environments/Sets/Concert/Textures/T_Club_Interior_Floor_1}.G
   ) {constA=-0.5, constB=3}) {maxDefault=1.5}
  n7 = OneMinus(None=VertexColor().B)
  n8 = Multiply(A=Multiply(A=LinearInterpolate(B=@n1.RGB, Alpha=@n4) {constA=1}, B=@n6), B=@n7)
  n9 = Multiply(A=@n7, B=@n6)
  n10 = Constant() {r=0.55}
  n11 = MaterialFunctionCall(Float3=ReflectionVectorWS()) {materialFunction=&BreakOutFloat3Components}
  n12 = Multiply(A=Constant() {r=0.7}, B=@n11.R) {constB=2}
  n13 = Multiply(A=@n11.G, B=Constant() {r=3}) {constB=2}
  n14 = Constant() {r=0.05}
  n15 = TextureSample(UVs=MaterialFunctionCall(X=Add(A=@n10, B=@n12), Y=Add(A=@n13, B=@n14)) {materialFunction=&MakeFloat2}.Result) {texture=&/T_StoreBGVista}
  n16 = ScalarParameter() {defaultValue=0.3, parameterName=MobileReflectionAmount}
  Add(A=Multiply(A=@n8, B=ScalarParameter() {parameterName=MobileAmbient}), B=Multiply(A=@n9, B=Multiply(A=@n15.RGB, B=@n16)))
  LinearInterpolate(A=ScalarParameter() {defaultValue=1, parameterName=MetalLow_1}, B=ScalarParameter() {defaultValue=1, parameterName=MetalHigh_1}, Alpha=@n9)
  Multiply(A=@n8, B=ScalarParameter() {parameterName=MobileDark})
  ScalarParameter() {parameterName=Spec}
  ScalarParameter() {parameterName=Rough}
  ShadingPathSwitch(Default=LinearInterpolate(A=Constant() {r=0.5}, B=Constant() {r=0.95}, Alpha=@n9)) {desc=Migrated from FeatureLevelSwitch}
  ShadingPathSwitch(Default=Constant()) {desc=Migrated from FeatureLevelSwitch}
  ShadingPathSwitch(Default=Constant() {r=0.075}) {desc=Migrated from FeatureLevelSwitch}
  ShadingPathSwitch(Default=Constant()) {desc=Migrated from FeatureLevelSwitch}
  ShadingPathSwitch(Default=@n8) {desc=Migrated from FeatureLevelSwitch}
 ]
}
