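# M_StoreBGWalls is a Default Lit material for the front-end store's background walls, built from scalar values with no textures.
# A static switch, on by default, multiplies a vertex-blue blend by a clamped vertical UV gradient, or falls back to a constant when off.
# Shading path switches carry a base color parameter and a constant, alongside separate mobile base color and ambient parameters that are multiplied together.
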Material /Game/FrontEnd/Store/Background/M_StoreBGWalls
graph {
 properties {opacityMaskClipValue=0.1, pixelDepthOffsetMode=PDOM_Legacy, usedShadingModels=Default Lit}
 nodes [
  Constant()
  Constant()
  Constant() {r=1}
  Multiply(
   A=ScalarParameter() {defaultValue=0.1, parameterName=BaseColorValueMobile}
   B=ScalarParameter() {defaultValue=0.2, parameterName=MobileAmbient}
  )
  ShadingPathSwitch(Default=ScalarParameter() {defaultValue=0.31, parameterName=BaseColorValue}) {desc=Migrated from FeatureLevelSwitch}
  ShadingPathSwitch(Default=Constant()) {desc=Migrated from FeatureLevelSwitch}
  n1 = Constant() {r=0.31}
  n2 = VertexColor()
  n3 = Constant() {r=0.44}
  n4 = Constant() {r=1}
  n5 = TextureCoordinate()
  n6 = Clamp(None=LinearInterpolate(A=Constant() {r=-2}, B=Constant() {r=1.2}, Alpha=ComponentMask(None=@n5) {g=true}))
  StaticSwitchParameter(True=Multiply(A=LinearInterpolate(B=@n1, Alpha=@n2.B) {constA=1}, B=LinearInterpolate(A=@n3, B=@n4, Alpha=@n6)), False=@n1) {defaultValue=true, parameterName=UseVertexColor}
 ]
}
