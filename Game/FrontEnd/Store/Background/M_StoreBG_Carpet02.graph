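# Masked, lit carpet material for the front-end store background, previewed on a mansion rug mesh.
# The item pattern's alpha blends two rotated, differently tiled carpet samples, producing a mask that darkens a red tint.
# Pattern-blended Fresnel terms, modulated by that mask, build a colored sheen scaled by EmsveTEST, which defaults to zero.
# Tile-noise and pattern normals combine through BlendAngleCorrectedNormals; leftover constants and test-named parameters remain.
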Material /Game/FrontEnd/Store/Background/M_StoreBG_Carpet02
graph {
 properties {
  blendMode=BLEND_Masked
  opacityMaskClipValue=0.1
  pixelDepthOffsetMode=PDOM_Legacy
  previewMesh=/Game/Environments/Props/Mansion/Meshes/Proto/Prop_Mansion_Rug01
  usedShadingModels=Default Lit
 }
 nodes [
  n1 = Add(A=TextureCoordinate() {uTiling=0.5, vTiling=1.25}, B=Constant() {r=0.1})
  Add(B=Multiply(
    A=TextureSample(UVs=@n1) {samplerType=SAMPLERTYPE_Normal, texture=/Game/Weapons/WeaponSkins/Textures/T_TileNoise_03_NRM}.RGB
    B=VectorParameter() {parameterName=Color555, defaultValue=<rgba 0 0 1 1>}.RGB
   ))
  Constant3Vector() {constant=<rgba 0.03 0.0098 0.0093 1>}
  Constant3Vector() {constant=<rgba 0.54 0.1521 0.1242 1>}
  Constant3Vector() {constant=<rgba 0.3 0.3 1 1>}
  Constant3Vector() {constant=<rgba 0.125 0.041 0.0388 1>}
  Constant() {r=1.75}
  Constant() {r=2.5}
  Constant() {r=0.6}
  Constant() {r=1.75}
  Constant() {r=1.5}
  Constant() {r=1.5}
  Constant() {r=0.3}
  Constant() {r=1.75}
  n2 = VectorParameter() {parameterName=ColorTTTT, defaultValue=<rgba 0.285 0.0367 0.0299 1>}
  n3 = Multiply(A=TextureCoordinate(), B=ScalarParameter() {defaultValue=1, parameterName=fgvb235fqa2})
  n4 = Constant() {r=1}
  n5 = Multiply(A=TextureCoordinate(), B=ScalarParameter() {defaultValue=3, parameterName=nnnnnergw5rt234.})
  n6 = Constant() {r=1}
  n7 = Add(A=Multiply(A=TextureCoordinate(), B=Constant() {r=4}), B=Constant() {r=0.25})
  n8 = TextureSample(UVs=@n7) {texture=/Game/UI/Style/Fortnite-Item-Pattern-512}
  n9 = LinearInterpolate(
   A=TextureSample(UVs=Rotator(Coordinate=@n3, Time=@n4) {speed=-40}) {samplerType=SAMPLERTYPE_LinearColor, texture=&/T_StoreCarpet}.R
   B=TextureSample(UVs=Rotator(Coordinate=@n5, Time=@n6) {speed=-45}) {samplerType=SAMPLERTYPE_LinearColor, texture=&T_StoreCarpet}.R
   Alpha=@n8.A
  )
  LinearInterpolate(
   A=@n2.RGB
   B=Multiply(A=@n2.RGB, B=ScalarParameter() {defaultValue=0.25, parameterName=Dark}) {constB=0.85}
   Alpha=Multiply(B=@n9) {constA=1}
  )
  n10 = TextureCoordinate() {uTiling=0.22}
  MaterialFunctionCall(
   BaseNormal=Multiply(
    A=TextureSample(UVs=@n10) {samplerType=SAMPLERTYPE_Normal, texture=&T_TileNoise_03_NRM}.RGB
    B=VectorParameter() {parameterName=ghj4rdski9997876, defaultValue=<rgba 0.3 0.3 1 1>}.RGB
   )
   AdditionalNormal=Multiply(
    A=TextureSample(UVs=@n7) {samplerType=SAMPLERTYPE_Normal, texture=&/T_Fortnite-Item-Pattern_N}.RGB
    B=VectorParameter() {parameterName=PatternNormalsAmount, defaultValue=<rgba 0.05 0.05 1 1>}.RGB
   )
  ) {materialFunction=&BlendAngleCorrectedNormals}
  n11 = VectorParameter() {parameterName=FresColor, defaultValue=<rgba 0.445 0.1254 0.1024 1>}
  n12 = LinearInterpolate(
   A=Fresnel(ExponentIn=ScalarParameter() {defaultValue=1.75, parameterName=Fres222}) {baseReflectFraction=0.005, exponent=8}
   B=Fresnel(ExponentIn=ScalarParameter() {defaultValue=1.5, parameterName=Fres}) {baseReflectFraction=0.005, exponent=8}
   Alpha=@n8.A
  )
  n13 = LinearInterpolate(
   A=ScalarParameter() {defaultValue=0.6, parameterName=dfg546dsdsd}
   B=ScalarParameter() {defaultValue=1.75, parameterName=44dfdssss}
   Alpha=@n9
  )
  Multiply(A=Multiply(A=LinearInterpolate(B=@n11.RGB, Alpha=@n12), B=Clamp(None=@n13)), B=ScalarParameter() {parameterName=EmsveTEST})
  Multiply(A=@n8.A)
  ScalarParameter() {defaultValue=1, parameterName=Rough}
  ScalarParameter() {parameterName=Spec}
  ScalarParameter() {parameterName=Metal}
  VectorParameter() {parameterName=Color, defaultValue=<rgba 0.03 0.0098 0.0093 1>}
 ]
}
