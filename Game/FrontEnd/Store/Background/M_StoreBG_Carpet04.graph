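# Lit material for a carpet surface in the front-end store background, overlaid with the Fortnite item pattern on UV1.
# Two fixed rotations of the carpet texture, chosen by pattern alpha, multiply an unrotated sample to shade the tint, with Fresnel modulation.
# A colored Fresnel sheen is broken up by carpet and fine tiled noise, scaled by an emissive amount.
# Noise and pattern normals combine through angle-corrected blending.
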Material /Game/FrontEnd/Store/Background/M_StoreBG_Carpet04
graph {
 properties {numCustomizedUVs=2, opacityMaskClipValue=0.1, pixelDepthOffsetMode=PDOM_Legacy, usedShadingModels=Default Lit}
 nodes [
  Add(A=Multiply(A=TextureCoordinate(), B=Constant() {r=4}), B=Constant() {r=0.25})
  n1 = ScalarParameter() {defaultValue=0.5, parameterName=EmAmt}
  n2 = Fresnel(ExponentIn=ScalarParameter() {defaultValue=2, parameterName=Fres}) {baseReflectFraction=0.005, exponent=8}
  n3 = Multiply(A=VectorParameter() {parameterName=FresColor, defaultValue=<rgba 0.1375 0.2402 1 1>}.RGB, B=@n2)
  n4 = TextureSample(UVs=Multiply(A=TextureCoordinate() {uTiling=1.15}, B=Constant() {r=5})) {samplerType=SAMPLERTYPE_LinearColor, texture=&/T_StoreCarpet}
  n5 = Clamp(None=LinearInterpolate(A=Constant() {r=-0.5}, B=Constant() {r=1.5}, Alpha=@n4.R))
  Multiply(A=@n1, B=Multiply(A=@n3, B=@n5))
  n6 = ScalarParameter() {defaultValue=0.8, parameterName=PatternBGBright}
  n7 = VectorParameter() {parameterName=Color, defaultValue=<rgba 0.185 0.185 0.185 1>}
  n8 = TextureCoordinate() {coordinateIndex=1}
  n9 = TextureSample(UVs=@n8) {texture=/Game/UI/Style/Fortnite-Item-Pattern-512}
  Multiply(A=LinearInterpolate(A=Multiply(A=@n6, B=@n7.RGB), B=@n7.RGB, Alpha=@n9.A), B=LinearInterpolate(Alpha=@n4.R) {constA=0.4})
  ScalarParameter() {parameterName=Metal}
  ScalarParameter() {defaultValue=1, parameterName=Rough}
  ScalarParameter() {parameterName=Spec}
  n10 = TextureSample(UVs=Multiply(A=TextureCoordinate(), B=ScalarParameter() {defaultValue=20, parameterName=FineCarpet})) {texture=/Game/Effects/Fort_Effects/Textures/Tile/T_TileNoise_03}
  n11 = LinearInterpolate(A=Constant() {r=0.2}, Alpha=Multiply(A=@n5, B=OneMinus(None=@n10.R)))
  ShadingPathSwitch(Default=Multiply(A=Multiply(A=@n3, B=@n11), B=@n1)) {desc=Migrated from FeatureLevelSwitch}
  n12 = TextureCoordinate()
  n13 = Constant() {r=1}
  n14 = LinearInterpolate(
   A=TextureSample(UVs=Rotator(Coordinate=@n12, Time=@n13) {speed=-40}) {samplerType=SAMPLERTYPE_LinearColor, texture=&T_StoreCarpet}.R
   B=TextureSample(UVs=Rotator(Coordinate=@n12, Time=@n13) {speed=-45}) {samplerType=SAMPLERTYPE_LinearColor, texture=&T_StoreCarpet}.R
   Alpha=@n9.A
  )
  n15 = Power(Base=Multiply(A=ComponentMask(None=@n14) {r=true}, B=@n4.R), Exp=Constant() {r=0.6}) {constExponent=0.5}
  n16 = Constant() {r=1}
  n17 = Multiply(A=@n7.RGB, B=LinearInterpolate(A=Constant() {r=0.5}, Alpha=LinearInterpolate(A=@n15, B=@n16, Alpha=@n2) {constA=1, constB=2}))
  ShadingPathSwitch(Default=LinearInterpolate(A=Multiply(A=@n6, B=@n17), B=@n17, Alpha=@n9.A)) {desc=Migrated from FeatureLevelSwitch}
  n18 = TextureSample(UVs=TextureCoordinate() {uTiling=0.22}) {samplerType=SAMPLERTYPE_Normal, texture=/Game/Weapons/WeaponSkins/Textures/T_TileNoise_03_NRM}
  n19 = Constant3Vector() {constant=<rgba 0.25 0.2 1 1>}
  n20 = TextureSample(UVs=@n8) {samplerType=SAMPLERTYPE_Normal, texture=&/T_Fortnite-Item-Pattern_N}
  n21 = Constant3Vector() {constant=<rgba 0.2 0.2 1 1>}
  ShadingPathSwitch(Default=MaterialFunctionCall(BaseNormal=Multiply(A=@n18.RGB, B=@n19), AdditionalNormal=Multiply(A=@n20.RGB, B=@n21)) {materialFunction=&BlendAngleCorrectedNormals}.Result) {desc=Migrated from FeatureLevelSwitch}
 ]
}
