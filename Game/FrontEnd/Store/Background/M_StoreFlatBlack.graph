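# M_StoreFlatBlack is a flat, near-black material, filed with the front-end store background assets.
# It uses the Default Lit shading model and contains only parameters: a BaseColor vector and Rough, Spec and Emissive scalars.
# The parameter defaults give a dark color, medium roughness, low specular and no emission, with no textures or math nodes.
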Material /Game/FrontEnd/Store/Background/M_StoreFlatBlack
graph {
 properties {opacityMaskClipValue=0.1, pixelDepthOffsetMode=PDOM_Legacy, usedShadingModels=Default Lit}
 nodes [
  ScalarParameter() {defaultValue=0.5, parameterName=Rough}
  ScalarParameter() {parameterName=Emissive}
  ScalarParameter() {defaultValue=0.1, parameterName=Spec}
  VectorParameter() {parameterName=BaseColor, defaultValue=<rgba 0.01 0.01 0.01 1>}
 ]
}
