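# Unlit material for the vista background of the front-end store, flagged for use with Nanite meshes.
# It samples the T_StoreBGVista texture with tiled UVs that repeat more often vertically than horizontally.
# The sampled RGB is brightened by a constant multiplier, then multiplied by a muted teal-grey constant tint.
# The graph is fully static, with no parameters or animation.
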Material /Game/FrontEnd/Store/Background/M_StoreVistaBG
graph {
 properties {
  bFullyRough=true
  bTangentSpaceNormal=false
  bUsedWithNanite=true
  pixelDepthOffsetMode=PDOM_Legacy
  shadingModel=MSM_Unlit
  usedShadingModels=Unlit
 }
 nodes [
  n1 = TextureCoordinate() {uTiling=3, vTiling=6}
  Multiply(
   A=Constant3Vector() {constant=<rgba 0.328 0.41 0.4059 1>}
   B=Multiply(A=TextureSample(UVs=@n1) {texture=&/T_StoreBGVista}.RGB, B=Constant() {r=3})
  )
 ]
}
